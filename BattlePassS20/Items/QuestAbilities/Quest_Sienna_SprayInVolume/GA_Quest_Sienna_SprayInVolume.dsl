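; Gameplay ability for the Battle Pass S20 Sienna "spray in volume" quest that keeps a quest gameplay effect on its owner while active.
; On activation from a gameplay event it calls the parent implementation, applies GE_Quest_Sienna_SprayInVolume to the owner and stores the handle.
; When the ability ends, cancelled or not, it removes that effect through the stored handle.
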(event Ability|EventActivateAbilityFromEvent
  (params (EventData :type Gameplay%20Event%20Data%20Structure%20%28by%20ref%29))
  (|CallParentFunction EventData)
  (bind _returnvalue (Ability|ApplyGameplayEffectToOwner "/BattlePassS20/Items/QuestAbilities/Quest_Sienna_SprayInVolume/GE_Quest_Sienna_SprayInVolume.GE_Quest_Sienna_SprayInVolume_C"))
  (Variables|Default|SetGE_Handle _returnvalue))

(event Ability|EventOnEndAbility
  (params (bWasCancelled :type Boolean))
  (Ability|RemoveGameplayEffectFromOwnerWithHandle (Variables|Default|GetGE_Handle)))
